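# Niagara environment effect for SevenBase, named RezHeal, that emits a slow trickle of small glowing blue cubes.
# Cubes001 spawns particles on a rotated torus shape, then pulls them by point attraction as they tumble and scale up and down.
# Mesh and sprite renderers share the particles, whose base RGB is zeroed while a separate GlowColor curve shifts from bright to deep blue.
# Empty001 has no renderers, and the system exposes no user parameters.
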NiagaraSystem /Game/Effects/Fort_Effects/Effects/Environment/Hera/SevenBase/NS_ENV_FX_SevenBase_RezHeal
dependencies {
 usedDataInterfaces [&NiagaraDataInterfaceCurve, &NiagaraDataInterfaceColorCurve]
 usedDynamicInputs []
 usedModules [&SystemState]
}
emitters {
 Cubes001 {
  emitterUpdateScript [
   &EmitterState()
   &SpawnRate(SpawnRate=2, Use Spawn Probability=false)
  ]
  particleSpawnScript [
   &InitializeParticle(
    Color=<rgba 0.046 0.8162 1.5 5>
    Lifetime Max=4
    Lifetime Mode=::NewEnumerator1
    Mesh Scale Mode=::NewEnumerator1
    Mesh Uniform Scale Max=1.1
    Mesh Uniform Scale Min=0.7
    Position Offset=<0 -5 0>
    Sprite Size Mode=::NewEnumerator3
    UsePositionOffset=true
    Write Material Random=true
   )
   &ShapeLocation(
    Handle Radius=2
    Large Radius=10
    MaskSpawnGroup=false
    Override Seed=false
    Rotation Axis=<0 1 0>
    Rotation Mode=::NewEnumerator1
    Shape Primitive=::NewEnumerator3
    Surface Distribution=1
    UseOwnerScale=true
   )
  ]
  particleUpdateScript [
   &PointAttractionForce(
    Attraction Strength=&RandomRangeFloat(Minimum=0.2, Maximum=1)
    Kill Radius=0.1
    Kill Within Radius=false
    Module.Use Falloff=true
    UseAttractorOffset=false
   )
   &SolveForcesAndVelocity(Clamp Velocity=false, Limit Acceleration=false)
   &ParticleState()
   &ScaleMeshSize(Scale Factor=&VectorFromFloat(Value=&FloatFromCurve(FloatCurve=DataInterface() {propertyValues {Curve {keys [
          {interpMode=RCIM_Cubic, leaveTangent=3.858, tangentMode=RCTM_Break, time=0, value=0}
          {interpMode=RCIM_Cubic, tangentMode=RCTM_Break, time=0.5, value=1.05}
          {arriveTangent=-4.3191, tangentMode=RCTM_User, time=1, value=0}
         ]}}})))
   &UpdateMeshOrientation(Rotation Rate=&RandomRangeFloat(Minimum=0.15, Maximum=0.3), Rotation Vector=<1 1 1>)
   &ScaleColor(
    Scale Alpha=&FloatFromCurve(FloatCurve=DataInterface() {propertyValues {Curve {keys [{interpMode=RCIM_Cubic, tangentMode=RCTM_Break, time=0.3, value=1}, <at 1 0>]}}})
    Scale RGB=<0 0 0>
    ScaleA=true
    ScaleRGB=true
   )
   /SetVariables(Particles.GlowColor=&ColorFromCurve(ColorCurve=DataInterface() {propertyValues {
       AlphaCurve {keys [<at 0.0011 0>,<at 0.4528 0.4 RCIM_Cubic>,<at 1 0 RCIM_Cubic>]}
       BlueCurve {keys [<at 0 2 RCIM_Cubic>,<at 1 0.4167 RCIM_Cubic>]}
       GreenCurve {keys [<at 0 0.8564 RCIM_Cubic>,<at 1 0.0401 RCIM_Cubic>]}
       RedCurve {keys [<at 0 0 RCIM_Cubic>,<at 1 0.0561 RCIM_Cubic>]}
      }})) {isSetParameters=true}
  ]
  rendererClasses [&NiagaraMeshRendererProperties, &NiagaraSpriteRendererProperties]
  renderers [
   {rendererClass=&NiagaraMeshRendererProperties, rendererIndex=0}
   {rendererClass=&NiagaraSpriteRendererProperties, rendererIndex=1}
  ]
 }
 Empty001 {rendererClasses [], renderers []}
}
summary {
 systemName=NS_ENV_FX_SevenBase_RezHeal
 emitters [
  {emitterName=Cubes001, rendererClasses [&NiagaraMeshRendererProperties, &NiagaraSpriteRendererProperties]}
  {emitterName=Empty001, rendererClasses []}
 ]
 userVariables []
}
