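# Niagara environment effect of slowly rising bubbles for a SevenBase tank, built from one Bubbles emitter with mesh and sprite renderers.
# Bubbles spawn steadily in a cylinder, then drift up under randomized positive-Z gravity, drag and curl noise.
# A region mask multiplies a grow-and-shrink lifetime curve for mesh scale, limiting bubbles to the tank volume.
# Particle scale and random values feed dynamic material parameters; color scaling is disabled.
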NiagaraSystem /Game/Effects/Fort_Effects/Effects/Environment/Hera/SevenBase/NS_ENV_FX_SevenBase_TankBubbles
dependencies {
 usedDataInterfaces [&NiagaraDataInterfaceSpriteRendererInfo, &NiagaraDataInterfaceCurve]
 usedDynamicInputs []
 usedModules [&SystemState]
}
emitters {Bubbles {
  emitterUpdateScript [
   &EmitterState(Life Cycle Mode=::NewEnumerator1, Loop Duration=10, UseLoopDelay=false)
   &SpawnRate(SpawnRate=3, Use Spawn Probability=false)
  ]
  particleSpawnScript [
   &InitializeParticle(
    Color Maximum=<rgba 0.3285 0.9774 1 1>
    Color Minimum=<rgba 0.5909 0.6191 1 1>
    Color Mode=::NewEnumerator2
    Lifetime Max=7
    Lifetime Min=5
    Lifetime Mode=::NewEnumerator1
    Mesh Scale Mode=::NewEnumerator1
    Mesh Uniform Scale Max=0.1
    Mesh Uniform Scale Min=0.02
    Sprite Size Mode=::NewEnumerator1
    Uniform Sprite Size Max=20
    Uniform Sprite Size Min=10
    UsePositionOffset=false
    Write Material Random=true
   )
   &SubUVAnimation(
    "Sprite Renderer "=DataInterface() {propertyValues {SpriteRenderer=&self:Bubbles_0.NiagaraSpriteRendererProperties_0}}
    Start Frame Offset:NiagaraInt32=&RandomRangeInt(Minimum=0, Maximum=4)
    UseEndFrame=false
    UseStartFrame=false
   )
   &CylinderLocation(Cylinder Radius=50, MaskSpawnGroup=false, Override Seed=false, UseOwnerScale=true)
   &DynamicMaterialParameters(
    Index 0 Param 2=&Multiply_Float(A=3, B=&MakeFloatFromVector(Vector=$Particles.Scale, Channel=::NewEnumerator0))
    Module.Param0WriteEnabled=false
    Module.Param1WriteEnabled=true
    Module.Param2WriteEnabled=false
    Module.Param3WriteEnabled=false
   )
   &AddVelocity(Velocity=&RandomVector(Vector Scale=&RandomRangeFloat(Minimum=0, Maximum=50)))
  ]
  particleUpdateScript [
   &ParticleState()
   &Drag(Drag=2, Use Linear Drag=true, Use Rotational Drag=false)
   &GravityForce(Gravity=&MakeVector(X=0, Y=0, Z=&RandomRangeFloat(Minimum=5, Maximum=50)))
   &CurlNoiseForce(Noise Strength=25, Pan Noise=false)
   &SolveForcesAndVelocity(Clamp Velocity=false, Limit Acceleration=false)
   &ScaleColor(
    Scale Alpha=&FloatFromCurve(
     FloatCurve=DataInterface() {propertyValues {Curve {keys [
         {interpMode=RCIM_Cubic, tangentMode=RCTM_User, time=0, value=0}
         {interpMode=RCIM_Cubic, tangentMode=RCTM_User, time=0.3, value=1.05}
         {arriveTangent=-0.2343, interpMode=RCIM_Cubic, leaveTangent=-0.2343, tangentMode=RCTM_User, time=0.99, value=1}
         <at 1 0 RCIM_Cubic>
        ]}}}
     Scale Curve=&RandomRangeFloat(Minimum=0.25, Maximum=0.5)
    )
    Scale RGB=&VectorFromFloat(Value=5)
    ScaleA=true
    ScaleRGB=true
   ) {enabled=false}
   &RegionMask(Region Dimensions:Vector3f=<240 240 700>, Region Type:ENiagaraDebugDrawShape=::NewEnumerator1, UseOffset=false, UseRotation=false)
   &ScaleMeshSize(Scale Factor=&Multiply_VectorByFloat(
     Vector=<1 1 1>
     Float=&FloatFromCurve(
      FloatCurve=DataInterface() {propertyValues {Curve {
         defaultValue=1
         keys [
          {arriveTangent=7.2503, interpMode=RCIM_Cubic, leaveTangent=7.2503, tangentMode=RCTM_User, time=0, value=0}
          {interpMode=RCIM_Cubic, tangentMode=RCTM_User, time=0.5, value=1}
          {arriveTangent=-6.8957, tangentMode=RCTM_User, time=1, value=0}
         ]
        }}}
      Scale Curve=&Multiply_Float(A=1, B=$Particles.RegionMask.Float)
     )
    ))
   &DynamicMaterialParameters(
    Index 3 Param 2=&MakeFloatFromVector(Vector=$Particles.Scale, Channel=::NewEnumerator0)
    Index 3 Param 4=&RandomRangeFloat(Minimum=0, Maximum=100)
    Module.Param0WriteEnabled3=false
    Module.Param1WriteEnabled3=true
    Module.Param2WriteEnabled3=false
    Module.Param3WriteEnabled3=true
    Write Parameter Index 0=false
    Write Parameter Index 3=true
   ) {module=DynamicMaterialParameters002}
  ]
  rendererClasses [&NiagaraMeshRendererProperties, &NiagaraSpriteRendererProperties]
  renderers [
   {rendererClass=&NiagaraMeshRendererProperties, rendererIndex=0}
   {rendererClass=&NiagaraSpriteRendererProperties, rendererIndex=1}
  ]
 }}
summary {
 systemName=NS_ENV_FX_SevenBase_TankBubbles
 emitters [{emitterName=Bubbles, rendererClasses [&NiagaraMeshRendererProperties, &NiagaraSpriteRendererProperties]}]
 userVariables []
}
